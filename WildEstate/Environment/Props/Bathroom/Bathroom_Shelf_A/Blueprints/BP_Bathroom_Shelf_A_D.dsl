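; Bathroom shelf prop Blueprint that gives its grouped shelf-items mesh the same material as its main static mesh component.
; The construction script runs the parent's logic, then creates a dynamic material instance on the main component's first slot.
; It assigns that same instance to the first slot of the shelf-items mesh, so both share one material.
; BeginPlay, overlap and Tick events are empty.
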(event EventBeginPlay)

(event Collision|EventActorBeginOverlap
  (params (OtherActor :type Actor%20Object%20Reference)))

(event EventTick
  (params (DeltaSeconds :type Float%20%28single-precision%29)))

(fn ConstructionScript
  (|CallParentFunction)
  (bind _returnvalue (Rendering|Material|CreateDynamicMaterialInstance (Variables|Debug|GetStaticMeshComponent) 0 0 "None" :!class /Script/Engine.PrimitiveComponent))
  (Rendering|Material|SetMaterial (Variables|Default|GetSM_Bathroom_ShelfItemsGrp_A) 0 _returnvalue))
